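# Niagara system that draws dark, red-tinted motion-blur ribbon trails for the WildEstate BananaBus from an animated skeletal mesh.
# One emitter spawns short-lived particles through shared AnimTrail functions that set bone indices, with user inputs for skeletal mesh and ribbon ID offset.
# Local init and update AnimTrail modules read a system static mesh.
# Curves fade particle alpha and a separate BackgroundColor over life, feeding two ribbon renderers.
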NiagaraSystem /WildEstate/Assets/FX/BananaBus/NS_WildEstate_BananaBus_MotionBlur
dependencies {
 usedDataInterfaces [&NiagaraDataInterfaceCurve, &NiagaraDataInterfaceStaticMesh]
 usedDynamicInputs [&SimulationPosition]
 usedModules [
  &SystemState
  &self:SystemSpawnScript.SystemScriptSource.SystemScriptGraph.NiagaraNodeAssignment_1.SetVariables_#1
 ]
}
emitters {MotionBlur {
  emitterUpdateScript [
   &EmitterState()
   &FN_SpawnAnimTrail()
  ]
  particleSpawnScript [
   &InitializeParticle(Color Mode=::NewEnumerator0, Lifetime=0.3, UsePositionOffset=false, Write Material Random=true)
   &FN_InitializeAnimTrailBoneIndecies()
   &self:Initialize_AnimTrail(Color:LinearColor=<rgba 0.007 0.0024 0.0024 1>, Static Mesh:NiagaraDataInterfaceStaticMesh=$System.Static Mesh) {module=Initialize_AnimTrail}
  ]
  particleUpdateScript [
   &ParticleState()
   &ScaleColor(
    Scale Alpha=&FloatFromCurve(FloatCurve=DataInterface() {propertyValues {Curve {keys [
         {arriveTangent=-0.0502, time=0.1408, value=1}
         {arriveTangent=0.0619, interpMode=RCIM_Cubic, leaveTangent=0.0619, tangentMode=RCTM_User, time=0.5, value=1}
         {arriveTangent=-5.9183, interpMode=RCIM_Cubic, leaveTangent=-5.9183, tangentMode=RCTM_User, time=1, value=0}
        ]}}})
    Scale RGB=&MakeVectorFromLinearColorRGB(Color=<rgba 1 1 1 1>)
    ScaleA=true
    ScaleRGB=true
   )
   /SetVariables(Particles.BackgroundColor:LinearColor=&MultiplyLinearColor(
     Linear Color=<rgba 0.0469 0.0089 0.0073 1>
     Scale Factor=&MakeFloat4FromVectorAndFloat(
      "Vector (RGB)"=<1 1 1>
      "Float (Alpha)"=&FloatFromCurve(
       FloatCurve=DataInterface() {propertyValues {Curve {keys [
           {arriveTangent=-0.2221, interpMode=RCIM_Cubic, leaveTangent=-0.2221, tangentMode=RCTM_User, time=0.1408, value=1}
           {arriveTangent=-2.1097, interpMode=RCIM_Cubic, leaveTangent=-2.1097, tangentMode=RCTM_User, time=1, value=0}
          ]}}}
       Scale Curve=4
      )
     )
    )) {isSetParameters=true}
   &self:Update_AnimTrail(Static Mesh:NiagaraDataInterfaceStaticMesh=$System.Static Mesh) {module=Update_AnimTrail}
  ]
  rendererClasses [&NiagaraRibbonRendererProperties]
  renderers [
   {rendererClass=&NiagaraRibbonRendererProperties}
   {rendererClass=&NiagaraRibbonRendererProperties, rendererIndex=1}
  ]
 }}
summary {
 systemName=NS_WildEstate_BananaBus_MotionBlur
 emitters [{emitterName=MotionBlur, rendererClasses [&NiagaraRibbonRendererProperties]}]
 userVariables [
  User.RibbonIDOffset {type=NiagaraInt32}
  User.SkeletalMesh {
   type=NiagaraDataInterfaceSkeletalMesh
   defaultValue {
    struct=NiagaraExt_VariableValue_DataInterface
    value {dataInterface=&self:NiagaraDataInterfaceSkeletalMesh_0}
   }
  }
 ]
}
